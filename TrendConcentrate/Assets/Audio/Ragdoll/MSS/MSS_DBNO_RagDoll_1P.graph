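# Looping mono MetaSound source for first-person DBNO ragdoll movement audio.
# A gated air loop follows absolute elevation speed, with pitch from overall speed; randomly chosen rolling loops crossfade on AssetSwitch.
# StartRoll starts self-retriggering character foley on two alternating, cross-releasing players, with speed driving retrigger delay, pitch and volume.
# The Mobile page omits foley, and the rolling pitch mapping is computed but unconnected.
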MetaSoundSource /TrendConcentrate/Assets/Audio/Ragdoll/MSS/MSS_DBNO_RagDoll_1P
inputs [
 AirLoop.Air Loop Volume {type=Float}
 AirLoop.MaxPitch {type=Float}
 AirLoop.MaxVolume {type=Float}
 AirLoop.MinPitch {type=Float}
 AirLoop.MinSpeed {type=Float}
 AirLoop.MinVolume {type=Float}
 AirLoop.Wave {type=WaveAsset}
 AssetSwitch {type=Trigger}
 CharacterFoley {type=WaveAsset:Array}
 Foley.MasterGain {type=Float}
 Foley.Pitch.Max {type=Float}
 Foley.Pitch.Min {type=Float}
 Foley.RetriggerTime.Max {type=Float}
 Foley.REtriggerTime.Min {type=Float}
 Foley.Volume.Max {type=Float}
 Foley.Volume.Min {type=Float}
 Foley.WavFadeTime {type=Time}
 Loop.MasterGain {type=Float}
 Movement.Speed {type=Float}
 Movement.SpeedAzimuth {type=Float}
 Movement.SpeedElevation {type=Float}
 RollingLoop.MaxVolume {type=Float}
 RollingLoop.MinVolume {type=Float}
 RollingLoop.PitchMax {type=Float}
 RollingLoop.PitchMin {type=Float}
 RollingLoop.Waves {type=WaveAsset:Array}
 StartRoll {type=Trigger}
 UE.Source.OnPlay {type=Trigger}
 VelocityLength {type=Float}
]
interfaces [Fort.Movement:1.0, UE.OutputFormat.Mono:1.0, UE.Source:1.1]
outputs ["UE.OutputFormat.Mono.Audio:0" {type=Audio}]
pages [
 Default {
  is_build_graph=true
  input_defaults {
   AirLoop.Air Loop Volume=0.8
   AirLoop.MaxPitch=1.2
   AirLoop.MaxVolume=1
   AirLoop.MinPitch=1
   AirLoop.MinSpeed=0.7
   AirLoop.MinVolume=0.2
   AirLoop.Wave=/Game/Athena/Sounds/Wind_Falling/AMB_Wind_Skydive_Vertical_Loop_01
   AssetSwitch=null
   CharacterFoley=null
   Foley.MasterGain=0.8
   Foley.Pitch.Max=1.4
   Foley.Pitch.Min=-1
   Foley.REtriggerTime.Min=0.3
   Foley.RetriggerTime.Max=0.75
   Foley.Volume.Max=1.8
   Foley.Volume.Min=0.5
   Foley.WavFadeTime=0.3
   Loop.MasterGain=2
   Movement.Speed=0
   Movement.SpeedAzimuth=0
   Movement.SpeedElevation=0
   RollingLoop.MaxVolume=2.5
   RollingLoop.MinVolume=0.5
   RollingLoop.PitchMax=2
   RollingLoop.PitchMin=-5
   StartRoll=null
   UE.Source.OnPlay=null
   VelocityLength=0
   RollingLoop.Waves []
  }
  nodes [
   n1 = InitVariable.Audio()
   n2 = MapRange.MapRange:Float(In=AbsoluteValue.Abs:Float(Input=^Movement.SpeedElevation).Value) {also {defaults {In Range B=50}}}
   n3 = VariableAccessor.Float(Variable=VariableMutator.Float(Variable=InitVariable.Float().Variable, Value=UE.InterpTo:Audio(Target=@n2.Out Value) {also {defaults {Interp Time=0.25}}}.Value).Variable)
   n4 = &MSP_OneshotGate(OneshotThreshold=^AirLoop.MinSpeed, Value=@n3.Value)
   ADSREnvelopeADSREnvelope_2 = ADSR Envelope.ADSR Envelope:Float(Trigger Attack=@n4.OnGateEnter, Trigger Release=@n4.OnGateExit) {also {defaults {Attack Time=2, Decay Time=0, Release Time=2, Sustain Level=1}}}
   n5 = UE.Reroute(Value=@ADSREnvelopeADSREnvelope_2.Out Envelope)
   n6 = InitVariable.Float()
   n7 = MapRange.MapRange:Float(In=^Movement.Speed) {also {defaults {In Range A=100, In Range B=1200}}}
   n8 = VariableAccessor.Float(Variable=VariableMutator.Float(Variable=InitVariable.Float().Variable, Value=UE.InterpTo:Audio(Target=@n7.Out Value).Value).Variable)
   n9 = VariableAccessor.Float(Variable=@n8.Variable)
   n10 = VariableAccessor.Float(Variable=@n9.Variable)
   n11 = MapRange.MapRange:Float(Out Range B=^AirLoop.MaxPitch, Out Range A=^AirLoop.MinPitch, In Range A=^AirLoop.MinSpeed, In=@n10.Value)
   WavePlayer_2 = UE.Wave Player:Mono(
    Wave Asset=^AirLoop.Wave
    Play=@ADSREnvelopeADSREnvelope_2.On Attack Triggered
    Stop=UE.Reroute(Value=@ADSREnvelopeADSREnvelope_2.On Done).Value
    Pitch Shift=VariableAccessor.Float(Variable=VariableMutator.Float(Variable=@n6.Variable, Value=@n11.Out Value).Variable).Value
   ) {also {defaults {Loop="<MetaSoundBuilderResult.SUCCEEDED: 0>"}, version [1,1]}}
   n12 = UE.Multiply:Audio by Float(
    AdditionalOperands=^AirLoop.Air Loop Volume
    PrimaryOperand=UE.Multiply:Audio by Float(AdditionalOperands=UE.Reroute(Value=@n5.Value).Value, PrimaryOperand=@WavePlayer_2.Out Mono).Out
   )
   n13 = InitVariable.Audio()
   ArrayRandomGet_0 = Array.Random Get:WaveAsset:Array(Next=^AssetSwitch, In Array=^RollingLoop.Waves) {also {defaults {Weights []}}}
   n14 = /FNGameplayCues/Sounds/Fort_Audio_Blueprints/SurfaceTypeSound/MSP_SurfaceLoopSwapper(PlayNextSound=@ArrayRandomGet_0.On Next, NextSoundToPlay=@ArrayRandomGet_0.Value) {also {defaults {CrossfadeTime=0.2}}}
   n15 = InitVariable.Float()
   n16 = MapRange.MapRange:Float(Out Range B=^RollingLoop.MaxVolume, Out Range A=^RollingLoop.MinVolume, In=@n9.Value) {also {defaults {In Range A=0.3}}}
   n17 = InitVariable.Float()
   n18 = MapRange.MapRange:Float(
    Out Range B=^AirLoop.MaxVolume
    In Range A=^AirLoop.MinSpeed
    Out Range A=^AirLoop.MinVolume
    In=VariableAccessor.Float(Variable=@n3.Variable).Value
   )
   AudioMixerAudioMixerMono2_1 = "AudioMixer.Audio Mixer (Mono, 2)"(
    In 1=VariableAccessor.Audio(Variable=VariableMutator.Audio(Variable=@n1.Variable, Value=@n12.Out).Variable).Value
    In 0=VariableAccessor.Audio(Variable=VariableMutator.Audio(Variable=@n13.Variable, Value=@n14.Out).Variable).Value
    Gain 0=VariableAccessor.Float(Variable=VariableMutator.Float(Variable=@n15.Variable, Value=@n16.Out Value).Variable).Value
    Gain 1=VariableAccessor.Float(Variable=VariableMutator.Float(Variable=@n17.Variable, Value=@n18.Out Value).Variable).Value
   )
   n19 = InitVariable.Audio()
   n20 = InitVariable.Trigger()
   n21 = VariableDeferredAccessor.Trigger(Variable=InitVariable.Trigger().Variable)
   TriggerAnyTriggerAny2_1 = "TriggerAny.Trigger Any (2)"(In 0=^StartRoll, In 1=@n21.Value)
   TriggerSequence2_0 = "UE.Trigger Sequence (2)"(In=@TriggerAnyTriggerAny2_1.Out)
   ArrayRandomGet_1 = Array.Random Get:WaveAsset:Array(
    In Array=^CharacterFoley
    Next=VariableAccessor.Trigger(Variable=VariableMutator.Trigger(Variable=@n20.Variable, Value=@TriggerSequence2_0.Out 0).Variable).Value
   )
   n22 = VariableDeferredAccessor.Trigger(Variable=InitVariable.Trigger().Variable)
   ADSREnvelopeADSREnvelope_0 = ADSR Envelope.ADSR Envelope:Float(Release Time=^Foley.WavFadeTime, Trigger Attack=@ArrayRandomGet_1.On Next, Trigger Release=@n22.Value) {also {defaults {Attack Time=0, Decay Time=0, Sustain Level=1}}}
   n23 = VariableAccessor.Float(Variable=@n10.Variable)
   n24 = VariableAccessor.Float(Variable=VariableMutator.Float(
     Variable=InitVariable.Float().Variable
     Value=MapRange.MapRange:Float(Out Range B=^Foley.Pitch.Max, Out Range A=^Foley.Pitch.Min, In=@n23.Value) {also {defaults {In Range A=0.3}}}.Out Value
    ).Variable)
   WavePlayer_0 = UE.Wave Player:Mono(
    Play=@ADSREnvelopeADSREnvelope_0.On Attack Triggered
    Stop=@ADSREnvelopeADSREnvelope_0.On Done
    Wave Asset=UE.Reroute(Value=UE.Reroute(Value=@ArrayRandomGet_1.Value).Value).Value
    Pitch Shift=VariableAccessor.Float(Variable=@n24.Variable).Value
   ) {also {version [1,1]}}
   n25 = UE.Multiply:Audio by Float(AdditionalOperands=UE.Reroute(Value=@ADSREnvelopeADSREnvelope_0.Out Envelope).Value, PrimaryOperand=@WavePlayer_0.Out Mono)
   n26 = InitVariable.Audio()
   n27 = InitVariable.Trigger()
   ArrayRandomGet_2 = Array.Random Get:WaveAsset:Array(
    In Array=^CharacterFoley
    Next=VariableAccessor.Trigger(Variable=VariableMutator.Trigger(Variable=@n27.Variable, Value=@TriggerSequence2_0.Out 1).Variable).Value
   ) {also {defaults {Weights []}}}
   n28 = VariableDeferredAccessor.Trigger(Variable=InitVariable.Trigger().Variable)
   ADSREnvelopeADSREnvelope_1 = ADSR Envelope.ADSR Envelope:Float(Release Time=^Foley.WavFadeTime, Trigger Attack=@ArrayRandomGet_2.On Next, Trigger Release=@n28.Value) {also {defaults {Attack Time=0, Decay Time=0, Sustain Level=1}}}
   WavePlayer_1 = UE.Wave Player:Mono(
    Play=@ADSREnvelopeADSREnvelope_1.On Attack Triggered
    Stop=@ADSREnvelopeADSREnvelope_1.On Done
    Wave Asset=UE.Reroute(Value=UE.Reroute(Value=@ArrayRandomGet_2.Value).Value).Value
    Pitch Shift=@n24.Value
   ) {also {version [1,1]}}
   n29 = UE.Multiply:Audio by Float(AdditionalOperands=UE.Reroute(Value=@ADSREnvelopeADSREnvelope_1.Out Envelope).Value, PrimaryOperand=@WavePlayer_1.Out Mono)
   AudioMixerAudioMixerMono2_2 = "AudioMixer.Audio Mixer (Mono, 2)"(
    In 0=VariableAccessor.Audio(Variable=VariableMutator.Audio(Variable=@n19.Variable, Value=@n25.Out).Variable).Value
    In 1=VariableAccessor.Audio(Variable=VariableMutator.Audio(Variable=@n26.Variable, Value=@n29.Out).Variable).Value
   )
   n30 = VariableMutator.Float(
    Variable=InitVariable.Float().Variable
    Value=MapRange.MapRange:Float(Out Range B=^Foley.Volume.Max, Out Range A=^Foley.Volume.Min, In=VariableAccessor.Float(Variable=@n23.Variable).Value).Out Value
   )
   AudioMixerAudioMixerMono2_0 = "AudioMixer.Audio Mixer (Mono, 2)"(
    Gain 0=^Loop.MasterGain
    In 0=@AudioMixerAudioMixerMono2_1.Out
    In 1=@AudioMixerAudioMixerMono2_2.Out
    Gain 1=UE.Multiply:Float(AdditionalOperands=^Foley.MasterGain, PrimaryOperand=VariableAccessor.Float(Variable=@n30.Variable).Value).Out
   )
   InitVariable.Bool()
   VariableAccessor.Float(Variable=VariableMutator.Float(
     Variable=InitVariable.Float().Variable
     Value=MapRange.MapRange:Float(Out Range B=^RollingLoop.PitchMax, Out Range A=^RollingLoop.PitchMin, In=@n8.Value).Out Value
    ).Variable)
   n31 = MapRange.MapRange:Float(Out Range B=^Foley.REtriggerTime.Min, Out Range A=^Foley.RetriggerTime.Max, In=^Movement.Speed) {also {defaults {In Range A=500, In Range B=1200}}}
   n32 = InitVariable.Trigger()
   TriggerAnyTriggerAny2_0 = "TriggerAny.Trigger Any (2)"(
    In 0=VariableAccessor.Trigger(Variable=VariableMutator.Trigger(Variable=@n22.Variable, Value=@WavePlayer_1.On Play).Variable).Value
    In 1=VariableAccessor.Trigger(Variable=VariableMutator.Trigger(Variable=@n28.Variable, Value=@WavePlayer_0.On Play).Variable).Value
   )
   n33 = UE.RandomFloat(
    Max=UE.Add:Float(AdditionalOperands=UE.ConversionTimeToFloat(In=^Foley.WavFadeTime).Out, PrimaryOperand=@n31.Out Value).Out
    Min=@n31.Out Value
    Next=VariableAccessor.Trigger(Variable=VariableMutator.Trigger(Variable=@n32.Variable, Value=@TriggerAnyTriggerAny2_0.Out).Variable).Value
   )
   TriggerDelay_0 = UE.Trigger Delay(Delay Time=UE.ConversionFloatToTime(In=@n33.Value).Out, In=@n33.On Next)
   VariableMutator.Trigger(Value=@TriggerDelay_0.Out, Variable=@n21.Variable)
  ]
  links [{dst="@UE.OutputFormat.Mono.Audio:0", src=AudioMixerAudioMixerMono2_0.Out}]
 }
 Mobile {
  nodes [
   n1 = InitVariable.Audio()
   n2 = MapRange.MapRange:Float(In=AbsoluteValue.Abs:Float(Input=^Movement.SpeedElevation).Value) {also {defaults {In Range B=50}}}
   n3 = VariableAccessor.Float(Variable=VariableMutator.Float(Variable=InitVariable.Float().Variable, Value=UE.InterpTo:Audio(Target=@n2.Out Value) {also {defaults {Interp Time=0.25}}}.Value).Variable)
   n4 = &MSP_OneshotGate(OneshotThreshold=^AirLoop.MinSpeed, Value=@n3.Value)
   ADSREnvelopeADSREnvelope_0 = ADSR Envelope.ADSR Envelope:Float(Trigger Attack=@n4.OnGateEnter, Trigger Release=@n4.OnGateExit) {also {defaults {Attack Time=2, Decay Time=0, Release Time=2, Sustain Level=1}}}
   n5 = UE.Reroute(Value=@ADSREnvelopeADSREnvelope_0.Out Envelope)
   n6 = InitVariable.Float()
   n7 = MapRange.MapRange:Float(In=^Movement.Speed) {also {defaults {In Range A=100, In Range B=1200}}}
   n8 = VariableAccessor.Float(Variable=VariableMutator.Float(Variable=InitVariable.Float().Variable, Value=UE.InterpTo:Audio(Target=@n7.Out Value).Value).Variable)
   n9 = VariableAccessor.Float(Variable=@n8.Variable)
   n10 = MapRange.MapRange:Float(
    Out Range B=^AirLoop.MaxPitch
    Out Range A=^AirLoop.MinPitch
    In Range A=^AirLoop.MinSpeed
    In=VariableAccessor.Float(Variable=@n9.Variable).Value
   )
   WavePlayer_0 = UE.Wave Player:Mono(
    Wave Asset=^AirLoop.Wave
    Play=@ADSREnvelopeADSREnvelope_0.On Attack Triggered
    Stop=UE.Reroute(Value=@ADSREnvelopeADSREnvelope_0.On Done).Value
    Pitch Shift=VariableAccessor.Float(Variable=VariableMutator.Float(Variable=@n6.Variable, Value=@n10.Out Value).Variable).Value
   ) {also {defaults {Loop="<MetaSoundBuilderResult.SUCCEEDED: 0>"}, version [1,1]}}
   n11 = UE.Multiply:Audio by Float(
    AdditionalOperands=^AirLoop.Air Loop Volume
    PrimaryOperand=UE.Multiply:Audio by Float(AdditionalOperands=UE.Reroute(Value=@n5.Value).Value, PrimaryOperand=@WavePlayer_0.Out Mono).Out
   )
   n12 = InitVariable.Audio()
   ArrayRandomGet_0 = Array.Random Get:WaveAsset:Array(Next=^AssetSwitch, In Array=^RollingLoop.Waves) {also {defaults {Weights []}}}
   n13 = &MSP_SurfaceLoopSwapper(PlayNextSound=@ArrayRandomGet_0.On Next, NextSoundToPlay=@ArrayRandomGet_0.Value) {also {defaults {CrossfadeTime=0.2}}}
   n14 = InitVariable.Float()
   n15 = MapRange.MapRange:Float(Out Range B=^RollingLoop.MaxVolume, Out Range A=^RollingLoop.MinVolume, In=@n9.Value) {also {defaults {In Range A=0.3}}}
   n16 = InitVariable.Float()
   n17 = MapRange.MapRange:Float(
    Out Range B=^AirLoop.MaxVolume
    In Range A=^AirLoop.MinSpeed
    Out Range A=^AirLoop.MinVolume
    In=VariableAccessor.Float(Variable=@n3.Variable).Value
   )
   AudioMixerAudioMixerMono2_1 = "AudioMixer.Audio Mixer (Mono, 2)"(
    In 1=VariableAccessor.Audio(Variable=VariableMutator.Audio(Variable=@n1.Variable, Value=@n11.Out).Variable).Value
    In 0=VariableAccessor.Audio(Variable=VariableMutator.Audio(Variable=@n12.Variable, Value=@n13.Out).Variable).Value
    Gain 0=VariableAccessor.Float(Variable=VariableMutator.Float(Variable=@n14.Variable, Value=@n15.Out Value).Variable).Value
    Gain 1=VariableAccessor.Float(Variable=VariableMutator.Float(Variable=@n16.Variable, Value=@n17.Out Value).Variable).Value
   )
   AudioMixerAudioMixerMono2_0 = "AudioMixer.Audio Mixer (Mono, 2)"(Gain 0=^Loop.MasterGain, In 0=@AudioMixerAudioMixerMono2_1.Out)
   InitVariable.Audio()
   InitVariable.Audio()
   InitVariable.Bool()
   InitVariable.Float()
   InitVariable.Float()
   InitVariable.Trigger()
   InitVariable.Trigger()
   InitVariable.Trigger()
   InitVariable.Trigger()
   InitVariable.Trigger()
   InitVariable.Trigger()
   VariableAccessor.Float(Variable=VariableMutator.Float(
     Variable=InitVariable.Float().Variable
     Value=MapRange.MapRange:Float(Out Range B=^RollingLoop.PitchMax, Out Range A=^RollingLoop.PitchMin, In=@n8.Value).Out Value
    ).Variable)
  ]
  links [{dst="@UE.OutputFormat.Mono.Audio:0", src=AudioMixerAudioMixerMono2_0.Out}]
 }
]
properties {
 assetImportData=&self:AssetImportData
 attenuationSettings=/Game/Sounds/Attenuation/Fort_Player_Footsteps_1P_Sprint_Athena_Att
 bLooping=true
 duration=10000
 sampleRate=48000
 soundClassObject=&Fort_Player_Foley_1P_Athena
 volume=2
 concurrencySet [/TrendConcentrate/Assets/Audio/Ragdoll/CS_DBNO_Roll_1P]
}
