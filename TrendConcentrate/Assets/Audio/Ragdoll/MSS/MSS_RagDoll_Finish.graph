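# One-shot mono MetaSound that plays a random finish sound for ragdoll.
# TriggerAudio, rather than On Play, chains a random pitch shift, a random pick from FinishSounds and a random gain before starting the wave player.
# The gain scales the player's output, and playback completion fires On Finished.
# It uses third-person Fortnite footstep attenuation, a friendly foley sound class and a DBNO roll concurrency set.
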MetaSoundSource /TrendConcentrate/Assets/Audio/Ragdoll/MSS/MSS_RagDoll_Finish
inputs [
 FinishSounds {type=WaveAsset:Array}
 TriggerAudio {type=Trigger}
 UE.Source.OnPlay {type=Trigger}
]
interfaces [UE.OutputFormat.Mono:1.0, UE.Source.OneShot:1.0, UE.Source:1.1]
outputs [
 "UE.OutputFormat.Mono.Audio:0" {type=Audio}
 UE.Source.OneShot.OnFinished {type=Trigger}
]
pages [Default {
  input_defaults {FinishSounds=null, TriggerAudio=null, UE.Source.OnPlay=null}
  nodes [
   n1 = UE.RandomFloat(Next=^TriggerAudio) {also {defaults {Max=0.5, Min=-0.5}}}
   ArrayRandomGet_0 = Array.Random Get:WaveAsset:Array(In Array=^FinishSounds, Next=@n1.On Next)
   n2 = UE.RandomFloat(Next=@ArrayRandomGet_0.On Next) {also {defaults {Max=1.2, Min=0.8}}}
   WavePlayer_0 = UE.Wave Player:Mono(Wave Asset=@ArrayRandomGet_0.Value, Pitch Shift=@n1.Value, Play=@n2.On Next) {also {version [1,1]}}
   Multiply_0 = UE.Multiply:Audio by Float(AdditionalOperands=@n2.Value, PrimaryOperand=@WavePlayer_0.Out Mono)
  ]
  links [{dst="@UE.OutputFormat.Mono.Audio:0", src=Multiply_0.Out}, {dst="@UE.Source.OneShot.OnFinished", src=WavePlayer_0.On Finished}]
 }]
properties {
 assetImportData=&self:AssetImportData
 attenuationSettings=/Game/Sounds/Attenuation/Fort_Player_Footsteps_3P_Sprint_Athena_Att
 maxDistance=3584
 sampleRate=48000
 soundClassObject=/Game/Sounds/FortSoundClassesAndModes/Fort_Player_Foley_3P_Friendly_Athena
 concurrencySet [/TrendConcentrate/Assets/Audio/Ragdoll/CS_DBNO_Roll_3p]
}
